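# Physics asset for the female medium Werewolf character, named as the collision set that a cape collides against.
# Capsules cover the torso, clavicles, head and feet, while arm, leg and neck_02 bodies have no shapes.
# Six five-bone jacket chains (front, side and back, left and right) also carry capsules and hang from the pelvis.
# Every constraint leaves swing and twist free, so the listed angular limits do not restrict motion.
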PhysicsAsset /BRCosmetics/Characters/Player/Female/Medium/Bodies/F_Med_Soldier_01/Meshes/F_MED_Werewolf_CapeColliders
bodies {
 calf_l []
 calf_r []
 clavicle_l [{shapeType=Capsule, length=7.6964, radius=6.3853, center=<4.0147 -0.8354 -1.7419>, rotation=<rot -77.5566 177.3915 7.7314>}]
 clavicle_r [{shapeType=Capsule, length=7.6964, radius=6.3853, center=<-4.0147 0.8354 1.7419>, rotation=<rot -77.5565 177.3915 -172.2685>}]
 dyn_jacketBack_01_l [{shapeType=Capsule, length=15, radius=10, shapeName=dyn_jacketBack_01_l_capsule, center=<5 0 -7>, rotation=<rot 90 0 0>}]
 dyn_jacketBack_01_r [{shapeType=Capsule, length=15, radius=10, shapeName=dyn_jacketBack_01_r_capsule, center=<5 0 -7>, rotation=<rot 90 0 0>}]
 dyn_jacketBack_02_l [{shapeType=Capsule, length=15, radius=10, shapeName=dyn_jacketBack_02_l_capsule, center=<5 0 -6>, rotation=<rot 90 0 0>}]
 dyn_jacketBack_02_r [{shapeType=Capsule, length=15, radius=10, shapeName=dyn_jacketBack_02_r_capsule, center=<5 0 -6>, rotation=<rot 90 0 0>}]
 dyn_jacketBack_03_l [{shapeType=Capsule, length=15, radius=10, shapeName=dyn_jacketBack_03_l_capsule, center=<5 0 -5>, rotation=<rot 90 0 0>}]
 dyn_jacketBack_03_r [{shapeType=Capsule, length=15, radius=10, shapeName=dyn_jacketBack_03_r_capsule, center=<5 0 -5>, rotation=<rot 90 0 0>}]
 dyn_jacketBack_04_l [{shapeType=Capsule, length=12.7082, radius=12.2918, shapeName=dyn_jacketBack_04_l_capsule, center=<5 0 -7.758>, rotation=<rot 90 0 0>}]
 dyn_jacketBack_04_r [{shapeType=Capsule, length=12.7082, radius=12.2918, shapeName=dyn_jacketBack_04_r_capsule, center=<5 0 -7.758>, rotation=<rot 90 0 0>}]
 dyn_jacketBack_05_l [{shapeType=Capsule, length=12.7082, radius=12.2918, shapeName=dyn_jacketBack_05_l_capsule, center=<5 0 -7.8691>, rotation=<rot 90 0 0>}]
 dyn_jacketBack_05_r [{shapeType=Capsule, length=12.7082, radius=12.2918, shapeName=dyn_jacketBack_05_r_capsule, center=<5 0 -7.8691>, rotation=<rot 90 0 0>}]
 dyn_jacketFront_01_l [{shapeType=Capsule, length=15, radius=10, shapeName=dyn_jacketFront_01_l_capsule, center=<5 0 -5>, rotation=<rot 90 0 0>}]
 dyn_jacketFront_01_r [{shapeType=Capsule, length=15, radius=10, shapeName=dyn_jacketFront_01_r_capsule, center=<5 0 -5>, rotation=<rot 90 0 0>}]
 dyn_jacketFront_02_l [{shapeType=Capsule, length=15, radius=10, shapeName=dyn_jacketFront_02_l_capsule, center=<5 0 -5>, rotation=<rot 90 0 0>}]
 dyn_jacketFront_02_r [{shapeType=Capsule, length=15, radius=10, shapeName=dyn_jacketFront_02_r_capsule, center=<5 0 -5>, rotation=<rot 90 0 0>}]
 dyn_jacketFront_03_l [{shapeType=Capsule, length=15, radius=10, shapeName=dyn_jacketFront_03_l_capsule, center=<5 0 -5>, rotation=<rot 90 0 0>}]
 dyn_jacketFront_03_r [{shapeType=Capsule, length=15, radius=10, shapeName=dyn_jacketFront_03_r_capsule, center=<5 0 -5>, rotation=<rot 90 0 0>}]
 dyn_jacketFront_04_l [{shapeType=Capsule, length=15, radius=10, shapeName=dyn_jacketFront_04_l_capsule, center=<5 0 -5>, rotation=<rot 90 0 0>}]
 dyn_jacketFront_04_r [{shapeType=Capsule, length=15, radius=10, shapeName=dyn_jacketFront_04_r_capsule, center=<5 0 -5>, rotation=<rot 90 0 0>}]
 dyn_jacketFront_05_l [{shapeType=Capsule, length=15, radius=10, shapeName=dyn_jacketFront_05_l_capsule, center=<5 0 -5>, rotation=<rot 90 0 0>}]
 dyn_jacketFront_05_r [{shapeType=Capsule, length=15, radius=10, shapeName=dyn_jacketFront_05_r_capsule, center=<5 0 -5>, rotation=<rot 90 0 0>}]
 dyn_jacketSide_01_l [{shapeType=Capsule, length=15, radius=10, shapeName=dyn_jacketSide_01_l_capsule, center=<5 4 -5>, rotation=<rot 90 0 0>}]
 dyn_jacketSide_01_r [{shapeType=Capsule, length=15, radius=10, shapeName=dyn_jacketSide_01_r_capsule, center=<5 -4 -5>, rotation=<rot 90 0 0>}]
 dyn_jacketSide_02_l [{shapeType=Capsule, length=15, radius=10, shapeName=dyn_jacketSide_02_l_capsule, center=<5 3 -5>, rotation=<rot 90 0 0>}]
 dyn_jacketSide_02_r [{shapeType=Capsule, length=15, radius=10, shapeName=dyn_jacketSide_02_r_capsule, center=<5 -3 -5>, rotation=<rot 90 0 0>}]
 dyn_jacketSide_03_l [{shapeType=Capsule, length=15, radius=10, shapeName=dyn_jacketSide_03_l_capsule, center=<5 0 -5>, rotation=<rot 90 0 0>}]
 dyn_jacketSide_03_r [{shapeType=Capsule, length=15, radius=10, shapeName=dyn_jacketSide_03_r_capsule, center=<5 0 -5>, rotation=<rot 90 0 0>}]
 dyn_jacketSide_04_l [{shapeType=Capsule, length=15, radius=10, shapeName=dyn_jacketSide_04_l_capsule, center=<5 0 -5>, rotation=<rot 90 0 0>}]
 dyn_jacketSide_04_r [{shapeType=Capsule, length=15, radius=10, shapeName=dyn_jacketSide_04_r_capsule, center=<5 0 -5>, rotation=<rot 90 0 0>}]
 dyn_jacketSide_05_l [{shapeType=Capsule, length=13.73, radius=11.27, shapeName=dyn_jacketSide_05_l_capsule, center=<5 0.0001 -5.9994>, rotation=<rot 90 0 0>}]
 dyn_jacketSide_05_r [{shapeType=Capsule, length=13.73, radius=11.27, shapeName=dyn_jacketSide_05_r_capsule, center=<5 0.0001 -5.9994>, rotation=<rot 90 0 0>}]
 foot_l [{shapeType=Capsule, length=12.5982, radius=6.5127, center=<-2.2072 6.0534 -0.3367>, rotation=<rot 0.0388 1.4701 -89.3982>}]
 foot_r [{shapeType=Capsule, length=12.5982, radius=6.5127, center=<2.2072 -6.0534 0.3367>, rotation=<rot 0.0388 1.4701 90.6018>}]
 head [{shapeType=Capsule, length=0, radius=18, center=<9.0697 0.0746 0>, rotation=<rot 0 -90 0>}]
 lowerarm_l []
 lowerarm_r []
 neck_02 []
 pelvis [{shapeType=Capsule, length=11.4279, radius=11.6054, center=<-1.9544 -2.2029 -0.1981>}]
 spine_01 [{shapeType=Capsule, length=9.9103, radius=11.7657, center=<-2.8183 -1.0873 0>}]
 spine_02 [
  {shapeType=Capsule, length=3.1898, radius=13.262, center=<1.179 2.2974 0>}
  {shapeType=Capsule, length=17.7456, radius=12.2544, center=<-1.8317 2.606 0>, rotation=<rot 90 0 0>}
 ]
 spine_03 [
  {shapeType=Capsule, length=3.5931, radius=12.2038, center=<-0.0747 2.3725 0>}
  {shapeType=Capsule, length=18.5187, radius=11.4813, center=<-1.2326 2.6059 0>, rotation=<rot 90 0 0>}
 ]
 spine_04 [
  {shapeType=Capsule, length=5.0366, radius=12.5263, center=<-0.9222 2.9298 0>}
  {shapeType=Capsule, length=10.0312, radius=9.2812, center=<-1.2326 2.6059 0>, rotation=<rot 90 0 0>}
 ]
 spine_05 [{shapeType=Capsule, length=9.7986, radius=11.2665, center=<-0.8027 1.4945 0>}]
 thigh_l []
 thigh_r []
 upperarm_l []
 upperarm_r []
}
constraints [
 {
  bone1Name=spine_01
  bone2Name=pelvis
  swing1LimitDegrees=45
  swing1Motion=Free
  swing2LimitDegrees=45
  swing2Motion=Free
  twistLimitDegrees=45
  twistMotion=Free
 }
 {
  bone1Name=spine_02
  bone2Name=spine_01
  swing1LimitDegrees=45
  swing1Motion=Free
  swing2LimitDegrees=45
  swing2Motion=Free
  twistLimitDegrees=45
  twistMotion=Free
 }
 {
  bone1Name=spine_03
  bone2Name=spine_02
  swing1LimitDegrees=45
  swing1Motion=Free
  swing2LimitDegrees=45
  swing2Motion=Free
  twistLimitDegrees=45
  twistMotion=Free
 }
 {
  bone1Name=spine_04
  bone2Name=spine_03
  swing1LimitDegrees=45
  swing1Motion=Free
  swing2LimitDegrees=45
  swing2Motion=Free
  twistLimitDegrees=45
  twistMotion=Free
 }
 {
  bone1Name=spine_05
  bone2Name=spine_04
  swing1LimitDegrees=45
  swing1Motion=Free
  swing2LimitDegrees=45
  swing2Motion=Free
  twistLimitDegrees=45
  twistMotion=Free
 }
 {
  bone1Name=clavicle_l
  bone2Name=spine_05
  swing1LimitDegrees=45
  swing1Motion=Free
  swing2LimitDegrees=45
  swing2Motion=Free
  twistLimitDegrees=45
  twistMotion=Free
 }
 {
  bone1Name=upperarm_l
  bone2Name=clavicle_l
  swing1LimitDegrees=45
  swing1Motion=Free
  swing2LimitDegrees=45
  swing2Motion=Free
  twistLimitDegrees=45
  twistMotion=Free
 }
 {
  bone1Name=lowerarm_l
  bone2Name=upperarm_l
  swing1LimitDegrees=45
  swing1Motion=Free
  swing2LimitDegrees=45
  swing2Motion=Free
  twistLimitDegrees=45
  twistMotion=Free
 }
 {
  bone1Name=neck_02
  bone2Name=spine_05
  swing1LimitDegrees=45
  swing1Motion=Free
  swing2LimitDegrees=45
  swing2Motion=Free
  twistLimitDegrees=45
  twistMotion=Free
 }
 {
  bone1Name=head
  bone2Name=neck_02
  swing1LimitDegrees=45
  swing1Motion=Free
  swing2LimitDegrees=45
  swing2Motion=Free
  twistLimitDegrees=45
  twistMotion=Free
 }
 {
  bone1Name=calf_l
  bone2Name=thigh_l
  swing1LimitDegrees=45
  swing1Motion=Free
  swing2LimitDegrees=45
  swing2Motion=Free
  twistLimitDegrees=45
  twistMotion=Free
 }
 {
  bone1Name=thigh_l
  bone2Name=pelvis
  swing1LimitDegrees=45
  swing1Motion=Free
  swing2LimitDegrees=45
  swing2Motion=Free
  twistLimitDegrees=45
  twistMotion=Free
 }
 {
  bone1Name=foot_l
  bone2Name=calf_l
  swing1LimitDegrees=45
  swing1Motion=Free
  swing2LimitDegrees=45
  swing2Motion=Free
  twistLimitDegrees=45
  twistMotion=Free
 }
 {
  bone1Name=clavicle_r
  bone2Name=spine_05
  swing1LimitDegrees=45
  swing1Motion=Free
  swing2LimitDegrees=45
  swing2Motion=Free
  twistLimitDegrees=45
  twistMotion=Free
 }
 {
  bone1Name=upperarm_r
  bone2Name=clavicle_r
  swing1LimitDegrees=45
  swing1Motion=Free
  swing2LimitDegrees=45
  swing2Motion=Free
  twistLimitDegrees=45
  twistMotion=Free
 }
 {
  bone1Name=lowerarm_r
  bone2Name=upperarm_r
  swing1LimitDegrees=45
  swing1Motion=Free
  swing2LimitDegrees=45
  swing2Motion=Free
  twistLimitDegrees=45
  twistMotion=Free
 }
 {
  bone1Name=thigh_r
  bone2Name=pelvis
  swing1LimitDegrees=45
  swing1Motion=Free
  swing2LimitDegrees=45
  swing2Motion=Free
  twistLimitDegrees=45
  twistMotion=Free
 }
 {
  bone1Name=calf_r
  bone2Name=thigh_r
  swing1LimitDegrees=45
  swing1Motion=Free
  swing2LimitDegrees=45
  swing2Motion=Free
  twistLimitDegrees=45
  twistMotion=Free
 }
 {
  bone1Name=foot_r
  bone2Name=calf_r
  swing1LimitDegrees=45
  swing1Motion=Free
  swing2LimitDegrees=45
  swing2Motion=Free
  twistLimitDegrees=45
  twistMotion=Free
 }
 {
  bone1Name=dyn_jacketBack_01_r
  bone2Name=pelvis
  swing1LimitDegrees=45
  swing1Motion=Free
  swing2LimitDegrees=45
  swing2Motion=Free
  twistLimitDegrees=45
  twistMotion=Free
 }
 {
  bone1Name=dyn_jacketBack_02_r
  bone2Name=dyn_jacketBack_01_r
  swing1LimitDegrees=45
  swing1Motion=Free
  swing2LimitDegrees=45
  swing2Motion=Free
  twistLimitDegrees=45
  twistMotion=Free
 }
 {
  bone1Name=dyn_jacketBack_03_r
  bone2Name=dyn_jacketBack_02_r
  swing1LimitDegrees=45
  swing1Motion=Free
  swing2LimitDegrees=45
  swing2Motion=Free
  twistLimitDegrees=45
  twistMotion=Free
 }
 {
  bone1Name=dyn_jacketBack_04_r
  bone2Name=dyn_jacketBack_03_r
  swing1LimitDegrees=45
  swing1Motion=Free
  swing2LimitDegrees=45
  swing2Motion=Free
  twistLimitDegrees=45
  twistMotion=Free
 }
 {
  bone1Name=dyn_jacketBack_05_r
  bone2Name=dyn_jacketBack_04_r
  swing1LimitDegrees=45
  swing1Motion=Free
  swing2LimitDegrees=45
  swing2Motion=Free
  twistLimitDegrees=45
  twistMotion=Free
 }
 {
  bone1Name=dyn_jacketBack_01_l
  bone2Name=pelvis
  swing1LimitDegrees=45
  swing1Motion=Free
  swing2LimitDegrees=45
  swing2Motion=Free
  twistLimitDegrees=45
  twistMotion=Free
 }
 {
  bone1Name=dyn_jacketBack_02_l
  bone2Name=dyn_jacketBack_01_l
  swing1LimitDegrees=45
  swing1Motion=Free
  swing2LimitDegrees=45
  swing2Motion=Free
  twistLimitDegrees=45
  twistMotion=Free
 }
 {
  bone1Name=dyn_jacketBack_03_l
  bone2Name=dyn_jacketBack_02_l
  swing1LimitDegrees=45
  swing1Motion=Free
  swing2LimitDegrees=45
  swing2Motion=Free
  twistLimitDegrees=45
  twistMotion=Free
 }
 {
  bone1Name=dyn_jacketBack_04_l
  bone2Name=dyn_jacketBack_03_l
  swing1LimitDegrees=45
  swing1Motion=Free
  swing2LimitDegrees=45
  swing2Motion=Free
  twistLimitDegrees=45
  twistMotion=Free
 }
 {
  bone1Name=dyn_jacketBack_05_l
  bone2Name=dyn_jacketBack_04_l
  swing1LimitDegrees=45
  swing1Motion=Free
  swing2LimitDegrees=45
  swing2Motion=Free
  twistLimitDegrees=45
  twistMotion=Free
 }
 {
  bone1Name=dyn_jacketSide_01_r
  bone2Name=pelvis
  swing1LimitDegrees=45
  swing1Motion=Free
  swing2LimitDegrees=45
  swing2Motion=Free
  twistLimitDegrees=45
  twistMotion=Free
 }
 {
  bone1Name=dyn_jacketSide_02_r
  bone2Name=dyn_jacketSide_01_r
  swing1LimitDegrees=45
  swing1Motion=Free
  swing2LimitDegrees=45
  swing2Motion=Free
  twistLimitDegrees=45
  twistMotion=Free
 }
 {
  bone1Name=dyn_jacketSide_03_r
  bone2Name=dyn_jacketSide_02_r
  swing1LimitDegrees=45
  swing1Motion=Free
  swing2LimitDegrees=45
  swing2Motion=Free
  twistLimitDegrees=45
  twistMotion=Free
 }
 {
  bone1Name=dyn_jacketSide_04_r
  bone2Name=dyn_jacketSide_03_r
  swing1LimitDegrees=45
  swing1Motion=Free
  swing2LimitDegrees=45
  swing2Motion=Free
  twistLimitDegrees=45
  twistMotion=Free
 }
 {
  bone1Name=dyn_jacketSide_05_r
  bone2Name=dyn_jacketSide_04_r
  swing1LimitDegrees=45
  swing1Motion=Free
  swing2LimitDegrees=45
  swing2Motion=Free
  twistLimitDegrees=45
  twistMotion=Free
 }
 {
  bone1Name=dyn_jacketSide_01_l
  bone2Name=pelvis
  swing1LimitDegrees=45
  swing1Motion=Free
  swing2LimitDegrees=45
  swing2Motion=Free
  twistLimitDegrees=45
  twistMotion=Free
 }
 {
  bone1Name=dyn_jacketSide_02_l
  bone2Name=dyn_jacketSide_01_l
  swing1LimitDegrees=45
  swing1Motion=Free
  swing2LimitDegrees=45
  swing2Motion=Free
  twistLimitDegrees=45
  twistMotion=Free
 }
 {
  bone1Name=dyn_jacketSide_03_l
  bone2Name=dyn_jacketSide_02_l
  swing1LimitDegrees=45
  swing1Motion=Free
  swing2LimitDegrees=45
  swing2Motion=Free
  twistLimitDegrees=45
  twistMotion=Free
 }
 {
  bone1Name=dyn_jacketSide_04_l
  bone2Name=dyn_jacketSide_03_l
  swing1LimitDegrees=45
  swing1Motion=Free
  swing2LimitDegrees=45
  swing2Motion=Free
  twistLimitDegrees=45
  twistMotion=Free
 }
 {
  bone1Name=dyn_jacketSide_05_l
  bone2Name=dyn_jacketSide_04_l
  swing1LimitDegrees=45
  swing1Motion=Free
  swing2LimitDegrees=45
  swing2Motion=Free
  twistLimitDegrees=45
  twistMotion=Free
 }
 {
  bone1Name=dyn_jacketFront_01_r
  bone2Name=pelvis
  swing1LimitDegrees=45
  swing1Motion=Free
  swing2LimitDegrees=45
  swing2Motion=Free
  twistLimitDegrees=45
  twistMotion=Free
 }
 {
  bone1Name=dyn_jacketFront_02_r
  bone2Name=dyn_jacketFront_01_r
  swing1LimitDegrees=45
  swing1Motion=Free
  swing2LimitDegrees=45
  swing2Motion=Free
  twistLimitDegrees=45
  twistMotion=Free
 }
 {
  bone1Name=dyn_jacketFront_03_r
  bone2Name=dyn_jacketFront_02_r
  swing1LimitDegrees=45
  swing1Motion=Free
  swing2LimitDegrees=45
  swing2Motion=Free
  twistLimitDegrees=45
  twistMotion=Free
 }
 {
  bone1Name=dyn_jacketFront_04_r
  bone2Name=dyn_jacketFront_03_r
  swing1LimitDegrees=45
  swing1Motion=Free
  swing2LimitDegrees=45
  swing2Motion=Free
  twistLimitDegrees=45
  twistMotion=Free
 }
 {
  bone1Name=dyn_jacketFront_05_r
  bone2Name=dyn_jacketFront_04_r
  swing1LimitDegrees=45
  swing1Motion=Free
  swing2LimitDegrees=45
  swing2Motion=Free
  twistLimitDegrees=45
  twistMotion=Free
 }
 {
  bone1Name=dyn_jacketFront_01_l
  bone2Name=pelvis
  swing1LimitDegrees=45
  swing1Motion=Free
  swing2LimitDegrees=45
  swing2Motion=Free
  twistLimitDegrees=45
  twistMotion=Free
 }
 {
  bone1Name=dyn_jacketFront_02_l
  bone2Name=dyn_jacketFront_01_l
  swing1LimitDegrees=45
  swing1Motion=Free
  swing2LimitDegrees=45
  swing2Motion=Free
  twistLimitDegrees=45
  twistMotion=Free
 }
 {
  bone1Name=dyn_jacketFront_03_l
  bone2Name=dyn_jacketFront_02_l
  swing1LimitDegrees=45
  swing1Motion=Free
  swing2LimitDegrees=45
  swing2Motion=Free
  twistLimitDegrees=45
  twistMotion=Free
 }
 {
  bone1Name=dyn_jacketFront_04_l
  bone2Name=dyn_jacketFront_03_l
  swing1LimitDegrees=45
  swing1Motion=Free
  swing2LimitDegrees=45
  swing2Motion=Free
  twistLimitDegrees=45
  twistMotion=Free
 }
 {
  bone1Name=dyn_jacketFront_05_l
  bone2Name=dyn_jacketFront_04_l
  swing1LimitDegrees=45
  swing1Motion=Free
  swing2LimitDegrees=45
  swing2Motion=Free
  twistLimitDegrees=45
  twistMotion=Free
 }
]
